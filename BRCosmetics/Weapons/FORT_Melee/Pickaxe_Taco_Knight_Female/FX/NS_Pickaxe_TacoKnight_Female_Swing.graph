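# Niagara swing effect for the female Taco Knight pickaxe, throwing off mesh food bits from a single emitter.
# Bits spawn in a squashed sphere offset from skeletal mesh samples, with spawn rate and upward velocity ramping over the loop.
# A random mesh index drives scale, tint and gravity, while drag, speed-driven tumbling and late shrinking finish the motion.
# Skydiving, running and landing spawners are present but disabled.
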NiagaraSystem /BRCosmetics/Weapons/FORT_Melee/Pickaxe_Taco_Knight_Female/FX/NS_Pickaxe_TacoKnight_Female_Swing
dependencies {usedDataInterfaces [&NiagaraDataInterfaceCurve], usedDynamicInputs [], usedModules [&SystemState]}
emitters {FoodBits {
  emitterUpdateScript [
   &EmitterState()
   &SpawnRate(
    Spawn Probability=0.1
    SpawnRate=&MakeCustomFloatFromBool(Bool=$User.bIsSkydiving, True Float=0, False Float=3)
    Use Spawn Probability=true
   ) {enabled=false}
   &SpawnRate(
    Spawn Probability=25
    SpawnRate=&Multiply_Float(
     A=&MakeCustomFloatFromBool(Bool=$User.bIsSkydiving, True Float=0, False Float=0.0025)
     B=$User.RunningSpeed
    )
    Use Spawn Probability=true
   ) {enabled=false, module=SpawnRate003}
   &SpawnRate(SpawnGroup=1, SpawnRate=&Multiply_Float(A=100, B=$User.Landing), Use Spawn Probability=false) {enabled=false, module=SpawnRate002}
   &SpawnRate(
    SpawnRate=&Multiply_Float(
     A=150
     B=&FloatFromCurve(FloatCurve=DataInterface() {propertyValues {Curve {keys [<at 0 0.2>,<at 1 1>]}}}, CurveIndex=$Emitter.NormalizedLoopAge)
    )
    Use Spawn Probability=true
   ) {module=SpawnRate001}
  ]
  particleSpawnScript [
   /SetVariables(Particles.MeshIndex=&RandomRangeInt(Minimum=0, Maximum=3)) {isSetParameters=true}
   &InitializeParticle(
    Color Mode=::NewEnumerator0
    Lifetime Max=1
    Lifetime Min=0.5
    Lifetime Mode=::NewEnumerator1
    Mass Max=2
    Mass Mode=::NewEnumerator3
    Mesh Scale=&Multiply_Vector(
     A=&MakeCustomVectorFromBool(
      Bool=&SetBoolByIntComparison(A=$Particles.MeshIndex, B=0, Comparison Type=::NewEnumerator4)
      True Vector=<1 1 1>
      False Vector=&MakeCustomVectorFromBool(
       Bool=&SetBoolByIntComparison(A=$Particles.MeshIndex, B=1, Comparison Type=::NewEnumerator4)
       True Vector=&RandomRangeVector(Minimum=<0.5 0.5 0.5>, Maximum=<1.5 1.5 1.5>)
       False Vector=&RandomRangeVector(Minimum=<0.3 0.3 0.3>, Maximum=<0.4 0.4 0.4>)
      )
     )
     B=&VectorFromFloat(Value=1.6)
    )
    Mesh Scale Mode=::NewEnumerator4
    Sprite Size Max=<3 14>
    Sprite Size Min=<2 6>
    Sprite Size Mode=::NewEnumerator2
    UsePositionOffset=false
    Write Material Random=true
   )
   &SphereLocation(
    MaskSpawnGroup=false
    Non Uniform Scale=<1 0.25 1>
    Offset=<0 -3 0>
    Override Seed=false
    Sphere Origin:NiagaraPosition=<0 0 0>
    Sphere Radius=17
    Surface Only=false
    UseNonUniformScale=true
    UseOffset=true
    UseSphereOrientation=false
   )
   &SkeletalMeshLocation(
    Inherit Velocity=false
    MaskSpawnGroup=false
    Mesh Sampling Type=::NewEnumerator5
    Sampled Position Offset:NiagaraPosition=$Output.SphereLocation.Position
    Skeletal Mesh=$User.Skeletal Mesh
    Socket Sampling Mode=::NewEnumerator5
   )
   &AddVelocityFromPoint(
    Module.Use Falloff Distance=false
    Origin Offset=<0 0 -2>
    Velocity Strength=&MakeCustomFloatFromBool(
     Bool=&SetBoolByIntComparison(A=&SpawnGroup(), B=0, Comparison Type=::NewEnumerator4)
     True Float=5
     False Float=75
    )
   ) {enabled=false}
   &InitialMeshOrientation(
    Rotation=&RandomRangeVector(Minimum=<0 0 0>, Maximum=<1 1 1>)
    Use Orientation Vector=false
    Use Rotation Vector=true
   )
   &Color(Color=&MakeCustomLinearColorFromBool(
     Bool=&SetBoolByIntComparison(A=$Particles.MeshIndex, B=0, Comparison Type=::NewEnumerator4)
     True Color=&RandomRangeLinearColor(Minimum=<rgba 0 0.2083 0.0834 1>, Maximum=<rgba 1 1 1 1>)
     False Color=&MakeCustomLinearColorFromBool(
      Bool=&SetBoolByIntComparison(A=$Particles.MeshIndex, B=1, Comparison Type=::NewEnumerator4)
      True Color=&RandomRangeLinearColor(Minimum=<rgba 1 0.6405 0.281 1>, Maximum=<rgba 1 1 1 1>, Color Channel Mode=::NewEnumerator2)
      False Color=&RandomRangeLinearColor(Minimum=<rgba 0.3333 0.2872 0.2014 1>, Maximum=<rgba 0.589 0.3147 0.2199 1>, Color Channel Mode=::NewEnumerator2)
     )
    ))
   &AddVelocity(Velocity=&Multiply_Vector(
     A=<0 0 500>
     B=&VectorFromFloat(Value=&FloatFromCurve(
       FloatCurve=DataInterface() {propertyValues {Curve {keys [
           {arriveTangent=0.9435, interpMode=RCIM_Cubic, leaveTangent=0.9435, tangentMode=RCTM_User, time=0, value=0}
           {arriveTangent=0.2155, interpMode=RCIM_Cubic, leaveTangent=0.2155, tangentMode=RCTM_User, time=1, value=1}
          ]}}}
       CurveIndex=$Emitter.NormalizedLoopAge
      ))
    ))
   &Color(Color=&MakeCustomLinearColorFromBool(
     Bool=&SetBoolByIntComparison(A=$Particles.MeshIndex, B=0, Comparison Type=::NewEnumerator4)
     True Color=&RandomRangeLinearColor(Minimum=<rgba 0 0.25 0 1>, Maximum=<rgba 0.684 2 0.0261 1>, Color Channel Mode=::NewEnumerator2)
     False Color=&MakeCustomLinearColorFromBool(
      Bool=&SetBoolByIntComparison(A=$Particles.MeshIndex, B=1, Comparison Type=::NewEnumerator4)
      True Color=&RandomRangeLinearColor(Minimum=<rgba 1 0 0 1>, Maximum=<rgba 2 0.7724 0 1>, Color Channel Mode=::NewEnumerator2)
      False Color=&RandomRangeLinearColor(Minimum=<rgba 0.8125 0.5918 0.0339 1>, Maximum=<rgba 0.3594 0.2548 0.189 1>, Color Channel Mode=::NewEnumerator2)
     )
    )) {enabled=false, module=Color001}
  ]
  particleUpdateScript [
   &ParticleState()
   &GravityForce(Gravity=&MakeCustomVectorFromBool(
     Bool=&SetBoolByIntComparison(A=$Particles.MeshIndex, B=1, Comparison Type=::NewEnumerator6)
     True Vector=<0 0 -800>
     False Vector=<0 0 -400>
    ))
   &Drag(Drag=&RandomRangeFloat(Minimum=1, Maximum=2), Use Linear Drag=true, Use Rotational Drag=false)
   &SolveForcesAndVelocity(Clamp Velocity=false, Limit Acceleration=false)
   &UpdateMeshOrientation(
    Delta Time:NiagaraFloat=&VectorLength(Vector=$Particles.Velocity)
    Rotation Rate=&RandomRangeFloat(Minimum=0.0003, Maximum=0.0005)
    Rotation Vector=&RandomRangeVector(Minimum=<0 0 0>, Maximum=<1 1 1>)
   )
   &ScaleMeshSize(Scale Factor=&VectorFromFloat(Value=&FloatFromCurve(FloatCurve=DataInterface() {propertyValues {Curve {keys [{interpMode=RCIM_Cubic, tangentMode=RCTM_User, time=0.5, value=1}, <at 1 0>]}}})))
  ]
  rendererClasses [&NiagaraMeshRendererProperties]
  renderers [{rendererClass=&NiagaraMeshRendererProperties, rendererIndex=0}]
 }}
summary {
 systemName=NS_Pickaxe_TacoKnight_Female_Swing
 emitters [{emitterName=FoodBits, rendererClasses [&NiagaraMeshRendererProperties]}]
 userVariables [
  User.bIsFrontEnd {type=NiagaraBool}
  User.bIsFrontEndPreview {type=NiagaraBool}
  User.bIsSkydiving {type=NiagaraBool}
  User.Landing {type=NiagaraFloat, defaultValue=0, description=""}
  User.RunningSpeed {type=NiagaraFloat, defaultValue=0, description=""}
  User.Skeletal Mesh {
   type=NiagaraDataInterfaceSkeletalMesh
   defaultValue {
    struct=NiagaraExt_VariableValue_DataInterface
    value {dataInterface=&self:NiagaraDataInterfaceSkeletalMesh_0}
   }
  }
 ]
}
