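# Physics asset for the Taco Knight Female pickaxe, giving capsule bodies to its dyn_ bones for secondary motion of the pepper and green parts.
# The pepper is a chain from dyn_base through a master bone with free swing and twist, then two segments with tight angular limits.
# Four small green bodies each constrain directly to dyn_base with moderate swing and narrow twist limits, forming a fan of independent pieces.
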PhysicsAsset /BRCosmetics/Weapons/FORT_Melee/Pickaxe_Taco_Knight_Female/Meshes/Taco_Knight_Female_Physics
bodies {
 dyn_base [{shapeType=Capsule, length=0.505, radius=0.505}]
 dyn_green_01 [{shapeType=Capsule, length=1, radius=1, center=<1 0 0>}]
 dyn_green_02 [{shapeType=Capsule, length=1, radius=1, center=<1 0 0>}]
 dyn_green_03 [{shapeType=Capsule, length=1, radius=1, center=<1 0 0>}]
 dyn_green_04 [{shapeType=Capsule, length=1, radius=1, center=<1 0 0>}]
 dyn_pepper_01 [{shapeType=Capsule, length=1.6624, radius=3.9204, center=<4 0 0>, rotation=<rot 90 0 0>}]
 dyn_pepper_02 [{shapeType=Capsule, length=3.1023, radius=3.6568, center=<5 0 0>, rotation=<rot 90 0 0>}]
 dyn_pepper_MSTR [{shapeType=Capsule, length=0.505, radius=0.505}]
}
constraints [
 {
  bone1Name=dyn_pepper_MSTR
  bone2Name=dyn_base
  swing1LimitDegrees=45
  swing1Motion=Free
  swing2LimitDegrees=45
  swing2Motion=Free
  twistLimitDegrees=45
  twistMotion=Free
 }
 {bone1Name=dyn_green_04, bone2Name=dyn_base, swing1LimitDegrees=25, swing2LimitDegrees=25, twistLimitDegrees=5}
 {bone1Name=dyn_green_03, bone2Name=dyn_base, swing1LimitDegrees=25, swing2LimitDegrees=25, twistLimitDegrees=5}
 {bone1Name=dyn_green_01, bone2Name=dyn_base, swing1LimitDegrees=25, swing2LimitDegrees=25, twistLimitDegrees=5}
 {bone1Name=dyn_green_02, bone2Name=dyn_base, swing1LimitDegrees=25, swing2LimitDegrees=25, twistLimitDegrees=5}
 {bone1Name=dyn_pepper_02, bone2Name=dyn_pepper_01, swing1LimitDegrees=10, swing2LimitDegrees=10, twistLimitDegrees=3}
 {bone1Name=dyn_pepper_01, bone2Name=dyn_pepper_MSTR, swing1LimitDegrees=10, swing2LimitDegrees=10, twistLimitDegrees=3}
]
